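# Physics asset for the RetroPhotographer female medium face accessory mesh, giving capsule bodies and joints to its hair buns and curls.
# Buns and side curls hang from the head with angular limits; R_curl2 chains off R_curl1 and the right bun's twist is locked.
# Five rear curls branch from back_curls through BASE bodies on free angular joints, with each curl tip limited in swing and tightly in twist.
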PhysicsAsset /BRCosmetics/Characters/Player/Female/Medium/Bodies/F_MED_RetroPhotographer/Meshes/Parts/F_MED_RetroPhotographer_FaceAcc_Physics
bodies {
 L_bun [{shapeType=Capsule, length=3.7, radius=6.2412, shapeName=L_bun_capsule, center=<-8.1035 0 0>, rotation=<rot 90 -90 -90>}]
 L_curl [{shapeType=Capsule, length=0.8934, radius=1.2847, shapeName=L_curl_capsule, center=<1.9756 0.076 0.0553>, rotation=<rot -4.0902 -90 -86.9938>}]
 R_bun [{shapeType=Capsule, length=2.9, radius=6.6073, shapeName=R_bun_capsule, center=<8.4445 3.5004 0>, rotation=<rot 18.3746 -0.0005 0.0001>}]
 R_curl [{shapeType=Capsule, length=1.005, radius=0.505, shapeName=R_curl_capsule, center=<1.1564 0.4657 0.023>, rotation=<rot 0 -72.8339 -87.8097>}]
 R_curl1 [{
   shapeType=Capsule
   length=0.505
   radius=0.505
   shapeName=R_curl1_capsule
   center=<1.0906 -0.128 0.1726>
   rotation=<rot 3.5679 -101.0794 -98.8458>
  }]
 R_curl2 [{
   shapeType=Capsule
   length=1.5777
   radius=0.5029
   shapeName=R_curl2_capsule
   center=<1.3804 -0.0496 -0.0635>
   rotation=<rot 0 -82.4197 -87.7473>
  }]
 back_curls [{shapeType=Capsule, length=0.505, radius=0.505, shapeName=back_curls_capsule, rotation=<rot 0 -90 0>}]
 curl_01 [{
   shapeType=Capsule
   length=8.3099
   radius=1.3056
   shapeName=curl_01_capsule
   center=<5.6936 -0.4546 1.8913>
   rotation=<rot 0 89.2867 -66.9541>
  }]
 curl_01_BASE [{shapeType=Capsule, length=0.505, radius=0.505, shapeName=curl_01_BASE_capsule, rotation=<rot 0 -90 0>}]
 curl_02 [{
   shapeType=Capsule
   length=1.0295
   radius=0.4537
   shapeName=curl_02_capsule
   center=<0.8172 0.1212 -0.0924>
   rotation=<rot 0 -90.1591 -81.9269>
  }]
 curl_02_BASE [{shapeType=Capsule, length=0.505, radius=0.505, shapeName=curl_02_BASE_capsule, rotation=<rot 0 -90 0>}]
 curl_03 [{
   shapeType=Capsule
   length=1.2738
   radius=0.6144
   shapeName=curl_03_capsule
   center=<0.9782 0.0912 -0.0881>
   rotation=<rot 0 87.4108 -86.3437>
  }]
 curl_03_BASE [{shapeType=Capsule, length=0.505, radius=0.505, shapeName=curl_03_BASE_capsule, rotation=<rot 0 -90 0>}]
 curl_04 [{
   shapeType=Capsule
   length=0.7877
   radius=1.3607
   shapeName=curl_04_capsule
   center=<2.2647 0.4539 0.0965>
   rotation=<rot 0 -93.3001 -86.9884>
  }]
 curl_04_BASE [{shapeType=Capsule, length=0.505, radius=0.505, shapeName=curl_04_BASE_capsule, rotation=<rot 0 -90 0>}]
 curl_05 [{shapeType=Capsule, length=5.8736, radius=3, shapeName=curl_05_capsule, center=<6.5011 -1.3492 -0.4133>, rotation=<rot 0 -102.6415 -83.8738>}]
 curl_05_BASE [{shapeType=Capsule, length=0.505, radius=0.505, shapeName=curl_05_BASE_capsule, rotation=<rot 0 -90 0>}]
 head [{shapeType=Capsule, length=3, radius=3.1208}]
}
constraints [
 {bone1Name=L_bun, bone2Name=head, swing1LimitDegrees=20, swing2LimitDegrees=40, twistLimitDegrees=5}
 {bone1Name=R_bun, bone2Name=head, swing1LimitDegrees=5, swing2LimitDegrees=13.64, twistLimitDegrees=1, twistMotion=Locked}
 {bone1Name=R_curl, bone2Name=head, swing1LimitDegrees=20, swing2LimitDegrees=10, twistLimitDegrees=5}
 {bone1Name=R_curl1, bone2Name=head, swing1LimitDegrees=20, swing2LimitDegrees=3, twistLimitDegrees=5}
 {bone1Name=R_curl2, bone2Name=R_curl1, swing1LimitDegrees=20, swing2LimitDegrees=10, twistLimitDegrees=5}
 {
  bone1Name=back_curls
  bone2Name=head
  swing1LimitDegrees=45
  swing1Motion=Free
  swing2LimitDegrees=45
  swing2Motion=Free
  twistLimitDegrees=45
  twistMotion=Free
 }
 {
  bone1Name=curl_01_BASE
  bone2Name=back_curls
  swing1LimitDegrees=45
  swing1Motion=Free
  swing2LimitDegrees=45
  swing2Motion=Free
  twistLimitDegrees=45
  twistMotion=Free
 }
 {bone1Name=curl_01, bone2Name=curl_01_BASE, swing1LimitDegrees=30, swing2LimitDegrees=10, twistLimitDegrees=2}
 {
  bone1Name=curl_02_BASE
  bone2Name=back_curls
  swing1LimitDegrees=45
  swing1Motion=Free
  swing2LimitDegrees=45
  swing2Motion=Free
  twistLimitDegrees=45
  twistMotion=Free
 }
 {bone1Name=curl_02, bone2Name=curl_02_BASE, swing1LimitDegrees=30, swing2LimitDegrees=10, twistLimitDegrees=2}
 {
  bone1Name=curl_03_BASE
  bone2Name=back_curls
  swing1LimitDegrees=45
  swing1Motion=Free
  swing2LimitDegrees=45
  swing2Motion=Free
  twistLimitDegrees=45
  twistMotion=Free
 }
 {bone1Name=curl_03, bone2Name=curl_03_BASE, swing1LimitDegrees=30, swing2LimitDegrees=10, twistLimitDegrees=2}
 {
  bone1Name=curl_04_BASE
  bone2Name=back_curls
  swing1LimitDegrees=45
  swing1Motion=Free
  swing2LimitDegrees=45
  swing2Motion=Free
  twistLimitDegrees=45
  twistMotion=Free
 }
 {bone1Name=curl_04, bone2Name=curl_04_BASE, swing1LimitDegrees=30, swing2LimitDegrees=10, twistLimitDegrees=2}
 {
  bone1Name=curl_05_BASE
  bone2Name=back_curls
  swing1LimitDegrees=45
  swing1Motion=Free
  swing2LimitDegrees=45
  swing2Motion=Free
  twistLimitDegrees=45
  twistMotion=Free
 }
 {bone1Name=curl_05, bone2Name=curl_05_BASE, swing1LimitDegrees=30, swing2LimitDegrees=10, twistLimitDegrees=2}
 {bone1Name=L_curl, bone2Name=head, swing1LimitDegrees=30, swing2LimitDegrees=10, twistLimitDegrees=45}
]
