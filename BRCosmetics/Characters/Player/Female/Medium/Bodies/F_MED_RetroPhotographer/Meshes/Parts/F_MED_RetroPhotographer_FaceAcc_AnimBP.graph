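# Animation Blueprint for the RetroPhotographer female face accessory, following the attached parent mesh.
# The pose copied from the parent passes through RigidBody physics, simulated relative to neck_01 with damped component motion, then a Control Rig to output.
# Each update derives head_up_FA from head twist against spine_05.
# That value drives additive curl_01–05 bone transforms, but this chain is unconnected to the output pose.
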AnimBlueprint /BRCosmetics/Characters/Player/Female/Medium/Bodies/F_MED_RetroPhotographer/Meshes/Parts/F_MED_RetroPhotographer_FaceAcc_AnimBP
target_skeleton=&/F_MED_RetroPhotographer_FaceAcc_Skeleton
graphs [
 AnimGraph {nodes [
   n1 = Get head_up_FA() -> (head_up_FA:double)
   n2 = "Transform (Modify) Bone - Bone: curl_01"(Alpha=@n1, Translation="0,-1,0", Rotation="0,60,50", Scale="1,1,1") {
    rotationMode=BMM_Additive
    translationMode=BMM_Additive
    alphaBoolBlend {blendOption=Linear}
    boneToModify {boneName=curl_01}
    rotation {roll=50, yaw=60}
    translation {y=-1}
   }
   n3 = "Transform (Modify) Bone - Bone: curl_02"(ComponentPose=@n2, Alpha=@n1, Translation="0,-1,0.5", Rotation="0,0,110", Scale="1,1,1") {
    rotationMode=BMM_Additive
    translationMode=BMM_Additive
    alphaBoolBlend {blendOption=Linear}
    boneToModify {boneName=curl_02}
    rotation {roll=110}
    translation {y=-1, z=0.5}
   }
   n4 = "Transform (Modify) Bone - Bone: curl_03"(ComponentPose=@n3, Alpha=@n1, Translation="0,-1,0.2", Rotation="0,0,90", Scale="1,1,1") {
    rotationMode=BMM_Additive
    translationMode=BMM_Additive
    alphaBoolBlend {blendOption=Linear}
    boneToModify {boneName=curl_03}
    rotation {roll=90}
    translation {y=-1, z=0.2}
   }
   n5 = "Transform (Modify) Bone - Bone: curl_04"(ComponentPose=@n4, Alpha=@n1, Translation="0,0,1", Rotation="0,0,90", Scale="1,1,1") {
    rotationMode=BMM_Additive
    translationMode=BMM_Additive
    alphaBoolBlend {blendOption=Linear}
    boneToModify {boneName=curl_04}
    rotation {roll=90}
    translation {z=1}
   }
   "Transform (Modify) Bone - Bone: curl_05"(ComponentPose=@n5, Alpha=@n1, Translation="0,0,0", Rotation="0,0,90", Scale="0.5,1,1") {
    rotationMode=BMM_Additive
    scaleMode=BMM_Additive
    translationMode=BMM_Additive
    alphaBoolBlend {blendOption=Linear}
    boneToModify {boneName=curl_05}
    rotation {roll=90}
    scale {x=0.5}
   }
   n6 = RigidBody(ComponentPose=Local To Component(LocalPose=Copy Pose From Mesh() {bUseAttachedParent=true})) {
    bClampLinearTranslationLimitToRefPose=true
    bOverrideWorldGravity=true
    overridePhysicsAsset=&/F_MED_RetroPhotographer_FaceAcc_Physics
    simulationSpace=BaseBoneSpace
    alphaBoolBlend {blendOption=Linear}
    baseBoneRef {boneName=neck_01}
    componentAppliedLinearAccClamp=<500 500 500>
    componentLinearAccScale=<0.3 0.3 0.3>
    componentLinearVelScale=<0.3 0.3 0.3>
    overrideWorldGravity {z=-980}
   }
   n7 = Control Rig(Source=Component To Local(ComponentPose=@n6)) {
    alphaBoolBlend {blendOption=Linear}
    controlRigAssetReference {blueprintRigClass=&/F_MED_RetroPhotographer_FaceAcc_CtrlRig.F_MED_RetroPhotographer_FaceAcc_CtrlRig_C}
   }
   Output Pose(Result=@n7) {layerGroup=None}
  ]}
 EventGraph {nodes [
   Event BlueprintUpdateAnimation()
   > Set head_up_FA(head_up_FA:double=GetAlphaFromReferenceTransformDelta(SkeletalMeshComponent=GetOwnerSkeletalMeshComponent(), BoneName=head, BaseName=spine_05, InputRange=(X=-10,Y=-30), TwistAxis="1,0,0")) -> (Output_Get:double)
   GetAlphaFromReferenceTransformDelta(BoneName=head, BaseName=spine_05, InputRange=(X=-10,Y=-40), TwistAxis="1,0,0")
  ]}
]
variables [head_up_FA {type=number}]
